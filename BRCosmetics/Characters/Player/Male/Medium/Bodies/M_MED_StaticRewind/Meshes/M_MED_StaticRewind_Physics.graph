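# Physics asset for the Fortnite-style male medium StaticRewind character body, giving a dangling hip cable simple secondary-motion physics.
# Small sphere and capsule bodies cover the belt, right thigh, hip cable and a larger cable-base sphere.
# A single constraint ties the cable to its base, locking swing 1 and twist while leaving swing 2 limited, so the cable swings mainly along one axis.
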PhysicsAsset /BRCosmetics/Characters/Player/Male/Medium/Bodies/M_MED_StaticRewind/Meshes/M_MED_StaticRewind_Physics
bodies {
 belt [{shapeType=Sphere, radius=1, shapeName=belt_sphere}]
 dyn_hip_cable [{shapeType=Capsule, length=5.5, radius=3, shapeName=dyn_hip_cable_capsule, center=<2.5166 0 -2.6719>, rotation=<rot 0 0.0002 90>}]
 dyn_hip_cable_Base [{shapeType=Sphere, radius=15, shapeName=dyn_hip_cable_Base_sphere}]
 thigh_r [{shapeType=Capsule, length=24, radius=6, shapeName=thigh_r_capsule, center=<10.1423 -0.0001 0.5979>, rotation=<rot 90 180 -179.9995>}]
}
constraints [{
  bone1Name=dyn_hip_cable
  bone2Name=dyn_hip_cable_Base
  swing1LimitDegrees=45
  swing1Motion=Locked
  swing2LimitDegrees=25
  twistLimitDegrees=45
  twistMotion=Locked
 }]
