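# Post-copy animation Blueprint for the StaticRewind male medium player body cosmetic, adding secondary motion to straps, laces, cables and wires.
# It copies the attached parent's pose into a Control Rig fed by movement and weapon states, including two flags computed each update.
# An Anim Dynamics chain then simulates the dangling bones and caches the result.
# The output uses RigidBody physics when bRigidBodyEnabled, otherwise Anim Dynamics on the hip cable.
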AnimBlueprint /BRCosmetics/Characters/Player/Male/Medium/Bodies/M_MED_StaticRewind/Meshes/M_MED_StaticRewind_AnimBP
target_skeleton=&/M_MED_StaticRewind_Skeleton
graphs [
 AnimGraph {nodes [
   n1 = Anim Dynamics - Bone: dyn_hip_cable(ComponentPose=Local To Component(LocalPose=Use cached pose 'PostAnimDynamics'()), Alpha=1, ExternalForce="0,0,0") {
    angularDampingOverride=0.99
    bOverrideAngularDamping=true
    bUseGravityOverride=true
    alphaBoolBlend {blendOption=Linear}
    boundBone {boneName=dyn_hip_cable}
    componentAppliedLinearAccClamp=<500 501 500>
    componentLinearAccScale=<0.1 0.1 0.1>
    componentLinearVelScale=<0.1 0.1 0.1>
    gravityOverride {z=-400}
    physicsBodyDefinitions [{
      boundBone {boneName=dyn_hip_cable}
      boxExtents=<5 10 10>
      constraintSetup {angularLimitsMax=<5 45 10>, angularLimitsMin=<-5 -5 -10>, angularTarget=<0 0 0>}
      localJointOffset=<1 0 0>
     }]
   }
   n2 = RigidBody(ComponentPose=Local To Component(LocalPose=Use cached pose 'PostAnimDynamics'())) {
    bOverrideWorldGravity=true
    overridePhysicsAsset=&/M_MED_StaticRewind_Physics
    alphaBoolBlend {blendOption=Linear}
    componentAppliedLinearAccClamp=<3000 3000 1000>
    componentLinearAccScale=<0.5 0.5 0.2>
    componentLinearVelScale=<0.5 0.5 0.2>
    overrideWorldGravity {z=-400}
   }
   Output Pose(Result=Blend Poses by bool(BlendPose_0=Component To Local(ComponentPose=@n2), BlendPose_1=Component To Local(ComponentPose=@n1), bActiveValue=Get bRigidBodyEnabled()) {blendType=Linear}) {layerGroup=None}
   n3 = Control Rig!(
    Source=Copy Pose From Mesh() {bUseAttachedParent=true}
    Alpha
    bIsParachuteOpened:bool=Get bIsParachuteOpened()
    bIsSkydiving:bool=Get bIsSkyDiving()
    bIsNotPickaxe:bool=Get bIsNotPickaxe() -> (bIsNotPickaxe:bool)
    bIsCrouching:bool=Get bIsCrouching()
    bIsJumpingOrFalling:bool=Get bIsJumpingOrFalling() -> (bIsJumpingOrFalling:bool)
   ) -> (Pose) {
    alphaBoolBlend {blendOption=Linear}
    controlRigAssetReference {blueprintRigClass=&/M_MED_StaticRewind_CtrlRig.M_MED_StaticRewind_CtrlRig_C}
   }
   n4 = Anim Dynamics - Bone: dyn_belt_tip(ComponentPose=Local To Component(LocalPose=@n3), Alpha=1, ExternalForce="0,0,0") {
    bUseGravityOverride=true
    alphaBoolBlend {blendOption=Linear}
    boundBone {boneName=dyn_belt_tip}
    componentAppliedLinearAccClamp=<3000 3000 3000>
    componentLinearAccScale=<0.3 0.3 0.3>
    componentLinearVelScale=<0.3 0.3 0.3>
    gravityOverride {z=-300}
    physicsBodyDefinitions [{
      boundBone {boneName=dyn_belt_tip}
      boxExtents=<10 10 10>
      constraintSetup {angularLimitsMax=<10 10 15>, angularLimitsMin=<-10 0 -5>, angularTarget=<0 0 0>}
      localJointOffset=<3 0 0>
     }]
   }
   n5 = Anim Dynamics - Bone: dyn_kneepad_l_strap(ComponentPose=@n4, Alpha=1, ExternalForce="0,0,0") {
    bUseGravityOverride=true
    simulationSpace=RootRelative
    alphaBoolBlend {blendOption=Linear}
    boundBone {boneName=dyn_kneepad_l_strap}
    componentAppliedLinearAccClamp=<3000 3000 3000>
    componentLinearAccScale=<0.4 0.4 0.4>
    componentLinearVelScale=<0.4 0.4 0.4>
    gravityOverride {z=-200}
    physicsBodyDefinitions [{
      boundBone {boneName=dyn_kneepad_l_strap}
      boxExtents=<5 10 10>
      constraintSetup {angularLimitsMax=<5 5 20>, angularLimitsMin=<-5 -15 -15>, angularTarget=<0 0 0>}
      localJointOffset=<3 0 0>
     }]
    relativeSpaceBone {boneName=pelvis}
   }
   n6 = Anim Dynamics - Bone: dyn_shoelace_l(ComponentPose=@n5, Alpha=1, ExternalForce="0,0,0") {
    angularDampingOverride=0.9
    bOverrideAngularDamping=true
    bUseGravityOverride=true
    simulationSpace=RootRelative
    alphaBoolBlend {blendOption=Linear}
    boundBone {boneName=dyn_shoelace_l}
    componentAppliedLinearAccClamp=<300 300 300>
    componentLinearAccScale=<0.1 0.1 0.1>
    componentLinearVelScale=<0.1 0.1 0.1>
    gravityOverride {z=-400}
    physicsBodyDefinitions [{
      boundBone {boneName=dyn_shoelace_l}
      boxExtents=<10 10 5>
      constraintSetup {angularLimitsMax=<5 45 5>, angularLimitsMin=<-5 -60 -45>, angularTarget=<0 0 0>}
      localJointOffset=<2 0 0>
     }]
    relativeSpaceBone {boneName=pelvis}
   }
   n7 = Anim Dynamics - Bone: dyn_shoelace_r(ComponentPose=@n6, Alpha=1, ExternalForce="0,0,0") {
    bUseGravityOverride=true
    simulationSpace=RootRelative
    alphaBoolBlend {blendOption=Linear}
    boundBone {boneName=dyn_shoelace_r}
    componentAppliedLinearAccClamp=<600 600 600>
    componentLinearAccScale=<0.1 0.1 0.1>
    componentLinearVelScale=<0.1 0.1 0.1>
    gravityOverride {z=-400}
    physicsBodyDefinitions [{
      boundBone {boneName=dyn_shoelace_r}
      boxExtents=<5 10 10>
      constraintSetup {angularLimitsMax=<5 45 0>, angularLimitsMin=<-5 -5 -90>, angularTarget=<0 0 0>}
      localJointOffset=<2 0 0>
     }]
    relativeSpaceBone {boneName=pelvis}
   }
   n8 = Anim Dynamics - Bone: dyn_cable_01(ComponentPose=@n7, Alpha=1, ExternalForce="0,0,0") {
    bUseGravityOverride=true
    alphaBoolBlend {blendOption=Linear}
    boundBone {boneName=dyn_cable_01}
    componentAppliedLinearAccClamp=<3000 3000 3000>
    componentLinearAccScale=<0.4 0.4 0.4>
    componentLinearVelScale=<0.4 0.4 0.4>
    gravityOverride {z=-200}
    physicsBodyDefinitions [{
      boundBone {boneName=dyn_cable_01}
      boxExtents=<5 10 10>
      constraintSetup {angularLimitsMax=<5 0 10>, angularLimitsMin=<-5 0 -30>, angularTarget=<0 0 0>}
      localJointOffset=<-3 0 0>
     }]
   }
   n9 = Anim Dynamics - Bone: dyn_rip_01(ComponentPose=@n8, Alpha=1, ExternalForce="0,0,0") {
    angularDampingOverride=0.85
    bOverrideAngularDamping=true
    bUseGravityOverride=true
    alphaBoolBlend {blendOption=Linear}
    boundBone {boneName=dyn_rip_01}
    componentAppliedLinearAccClamp=<1000 1000 1000>
    componentLinearAccScale=<0.2 0.2 0.2>
    componentLinearVelScale=<0.2 0.2 0.2>
    gravityOverride {z=-50}
    physicsBodyDefinitions [{
      boundBone {boneName=dyn_rip_01}
      boxExtents=<5 10 10>
      constraintSetup {angularLimitsMax=<5 40 10>, angularLimitsMin=<-5 -30 -10>, angularTarget=<0 0 0>}
      localJointOffset=<4 0 0>
     }]
   }
   n10 = Anim Dynamics - Bone: dyn_rip_02(ComponentPose=@n9, Alpha=1, ExternalForce="0,0,0") {
    angularDampingOverride=0.85
    bOverrideAngularDamping=true
    bUseGravityOverride=true
    alphaBoolBlend {blendOption=Linear}
    boundBone {boneName=dyn_rip_02}
    componentAppliedLinearAccClamp=<1000 1000 1000>
    componentLinearAccScale=<0.2 0.2 0.2>
    componentLinearVelScale=<0.2 0.2 0.2>
    gravityOverride {z=-50}
    physicsBodyDefinitions [{
      boundBone {boneName=dyn_rip_02}
      boxExtents=<5 10 10>
      constraintSetup {angularLimitsMax=<5 40 10>, angularLimitsMin=<-5 -30 -10>, angularTarget=<0 0 0>}
      localJointOffset=<4 0 0>
     }]
   }
   n11 = Anim Dynamics - Bone: dyn_rip_03(ComponentPose=@n10, Alpha=1, ExternalForce="0,0,0") {
    angularDampingOverride=0.85
    bOverrideAngularDamping=true
    bUseGravityOverride=true
    alphaBoolBlend {blendOption=Linear}
    boundBone {boneName=dyn_rip_03}
    componentAppliedLinearAccClamp=<1000 1000 1000>
    componentLinearAccScale=<0.2 0.2 0.2>
    componentLinearVelScale=<0.2 0.2 0.2>
    gravityOverride {z=-50}
    physicsBodyDefinitions [{
      boundBone {boneName=dyn_rip_03}
      boxExtents=<5 10 10>
      constraintSetup {angularLimitsMax=<5 40 10>, angularLimitsMin=<-5 -30 -10>, angularTarget=<0 0 0>}
      localJointOffset=<4 0 0>
     }]
   }
   n12 = Anim Dynamics - Bone: dyn_rip_04(ComponentPose=@n11, Alpha=1, ExternalForce="0,0,0") {
    angularDampingOverride=0.85
    bOverrideAngularDamping=true
    bUseGravityOverride=true
    alphaBoolBlend {blendOption=Linear}
    boundBone {boneName=dyn_rip_04}
    componentAppliedLinearAccClamp=<1000 1000 1000>
    componentLinearAccScale=<0.2 0.2 0.2>
    componentLinearVelScale=<0.2 0.2 0.2>
    gravityOverride {z=-50}
    physicsBodyDefinitions [{
      boundBone {boneName=dyn_rip_04}
      boxExtents=<5 10 10>
      constraintSetup {angularLimitsMax=<5 40 10>, angularLimitsMin=<-5 -30 -10>, angularTarget=<0 0 0>}
      localJointOffset=<4 0 0>
     }]
   }
   n13 = Anim Dynamics - Bone: dyn_wire_01(ComponentPose=@n12, Alpha=1, ExternalForce="0,0,0") {
    angularDampingOverride=0.9
    bOverrideAngularDamping=true
    bUseGravityOverride=true
    simulationSpace=RootRelative
    alphaBoolBlend {blendOption=Linear}
    boundBone {boneName=dyn_wire_01}
    componentAppliedLinearAccClamp=<3000 3000 3000>
    componentLinearAccScale=<0.2 0.2 0.2>
    componentLinearVelScale=<0.2 0.2 0.2>
    gravityOverride {z=-600}
    physicsBodyDefinitions [{
      boundBone {boneName=dyn_wire_01}
      boxExtents=<5 10 10>
      constraintSetup {angularLimitsMax=<0 30 30>, angularLimitsMin=<0 -30 -30>, angularTarget=<0 0 0>}
      localJointOffset=<1.5 0 0>
     }]
   }
   n14 = Anim Dynamics - Bone: dyn_wire_02(ComponentPose=@n13, Alpha=1, ExternalForce="0,0,0") {
    bUseGravityOverride=true
    simulationSpace=RootRelative
    alphaBoolBlend {blendOption=Linear}
    boundBone {boneName=dyn_wire_02}
    componentAppliedLinearAccClamp=<3000 3000 3000>
    componentLinearAccScale=<0.15 0.15 0.15>
    componentLinearVelScale=<0.15 0.15 0.15>
    gravityOverride {z=-300}
    physicsBodyDefinitions [{
      boundBone {boneName=dyn_wire_02}
      boxExtents=<5 10 10>
      constraintSetup {angularLimitsMax=<0 30 30>, angularLimitsMin=<0 -30 -30>, angularTarget=<0 0 0>}
      localJointOffset=<1 0 0>
     }]
   }
   Save cached pose 'PostAnimDynamics'(Pose=Component To Local(ComponentPose=@n14)) {cacheName=PostAnimDynamics}
  ]}
 EventGraph {nodes [
   Event BlueprintUpdateAnimation()
   > Set bIsNotPickaxe(bIsNotPickaxe:bool="Not Equal (Enum)"!(A:EFortWeaponCoreAnimation Enum=Get WeaponCoreAnim!() -> (WeaponCoreAnim), B:EFortWeaponCoreAnimation Enum=Melee) -> (ReturnValue:bool)) -> (Output_Get:bool)
   > Set bIsJumpingOrFalling(bIsJumpingOrFalling:bool=OR Boolean(A=Get bIsFalling(), B=Get bIsJumping())) -> (Output_Get:bool)
  ]}
]
variables [
 bIsJumpingOrFalling {type=boolean}
 bIsNotPickaxe {type=boolean}
]
